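# Niagara effect for the Meander emote that fires a brief camera flash.
# A single CameraFlash emitter burst-spawns short-lived particles at a fixed offset, drawn through sprite, light and component renderers.
# A float curve over particle life spikes then decays as FlashIntensity, driving sprite alpha and scaled point-light and spot-light values.
# The point-light position is shifted from the particle along negative Y.
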NiagaraSystem /BRCosmetics/Animation/Game/MainPlayer/Emotes/Meander/FX/NS_Emote_Meander_CameraFlash
dependencies {usedDataInterfaces [&NiagaraDataInterfaceCurve], usedDynamicInputs [], usedModules [&SystemState]}
emitters {CameraFlash {
  emitterUpdateScript [
   &EmitterState()
   &SpawnBurst_Instantaneous(Use Spawn Probability=false, UseLoopCountLimit=false)
  ]
  particleSpawnScript [&InitializeParticle(
    Color Mode=::NewEnumerator0
    Lifetime=0.2
    Position Offset=<0 120 115>
    Sprite Size Mode=::NewEnumerator3
    Uniform Sprite Size=133
    UsePositionOffset=true
    Write Material Random=true
   )]
  particleUpdateScript [
   /SetVariables(Particles.FlashIntensity:NiagaraFloat=&FloatFromCurve(FloatCurve=DataInterface() {propertyValues {Curve {keys [
         {arriveTangent=9.0725, time=0, value=0}
         {arriveTangent=-23.1883, leaveTangent=-3.6307, time=0.1, value=1}
         <at 0.2 0.325>
         <at 0.5 0.05>
         {arriveTangent=0.0076, leaveTangent=-0.0405, time=1, value=0}
        ]}}})) {isSetParameters=true}
   &ParticleState()
   /SetVariables(
    Particles.Color=&MakeLinearColorFromVectorAndFloat("Vector (RGB)"=<1 1 1>, "Float (Alpha)"=$StackContext.FlashIntensity)
    Particles.PointLightColor:LinearColor=&MakeLinearColorFromVectorAndFloat("Vector (RGB)"=<1 1 1>, "Float (Alpha)"=&Multiply_Float(A=$StackContext.FlashIntensity, B=100))
    Particles.PointLightPosition:NiagaraPosition=&AddVectorToPosition(Position=$Particles.Position, Vector=<0 -50 0>)
    Particles.SpotLightExponent=2
    Particles.SpotLightIntensity:NiagaraFloat=&Multiply_Float(A=$StackContext.FlashIntensity, B=150)
   ) {isSetParameters=true}
  ]
  rendererClasses [
   &NiagaraComponentRendererProperties
   &NiagaraSpriteRendererProperties
   &NiagaraLightRendererProperties
  ]
  renderers [
   {rendererClass=&NiagaraComponentRendererProperties, rendererIndex=0}
   {rendererClass=&NiagaraSpriteRendererProperties, rendererIndex=1}
   {rendererClass=&NiagaraLightRendererProperties, rendererIndex=2}
  ]
 }}
summary {
 systemName=NS_Emote_Meander_CameraFlash
 emitters [{
   emitterName=CameraFlash
   rendererClasses [
    &NiagaraComponentRendererProperties
    &NiagaraSpriteRendererProperties
    &NiagaraLightRendererProperties
   ]
  }]
 userVariables []
}
